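# Post-copy corrective AnimBlueprint for the M_MED_Plum_Retro player body cosmetic, adding procedural upper-body muscle deformation to the pose copied from the attached parent mesh.
# Constraint and Look At nodes aim pectoral, scapula and lat bones, then additive bone transforms deform triceps, lats and right biceps/pec.
# EventGraph derives their alphas from distances between measure bones.
# Left lat alpha drives right-side bones; left-pec and trapezius nodes are unconnected to the output.
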AnimBlueprint /BRCosmetics/Characters/Player/Male/Medium/Bodies/M_MED_Plum_Retro/Meshes/M_MED_Plum_Retro_AnimBP
target_skeleton=&/M_MED_Plum_Retro_Skeleton
graphs [
 AnimGraph {nodes [
   n1 = Look At - Bone: dyn_trapezius_01_r(Alpha=1) {
    alphaBoolBlend {blendOption=Linear}
    boneToModify {boneName=dyn_trapezius_01_r}
    lookAtLocation {x=0}
    lookAtTarget {boneReference {boneName=dyn_trapezius_02_r_aim}}
    lookAt_Axis {axis=<1 0>}
    lookUp_Axis {axis {x=0, z=1}}
   }
   Look At - Bone: dyn_trapezius_01_l(ComponentPose=@n1, Alpha=1) {
    alphaBoolBlend {blendOption=Linear}
    boneToModify {boneName=dyn_trapezius_01_l}
    lookAtLocation {x=0}
    lookAtTarget {boneReference {boneName=dyn_trapezius_02_l_aim}}
    lookAt_Axis {axis=<1 0>}
    lookUp_Axis {axis {x=0, z=1}}
   }
   n2 = Get LeftPecStretchAlpha() -> (LeftPecStretchAlpha:double)
   n3 = "Transform (Modify) Bone - Bone: bicep_01_l"(Alpha=@n2, Translation="0,0,0", Rotation="0,0,0", Scale="0.15,0.15,0.15") {
    rotationMode=BMM_Additive
    rotationSpace=BCS_BoneSpace
    scaleMode=BMM_Additive
    scaleSpace=BCS_BoneSpace
    alphaBoolBlend {blendOption=Linear}
    boneToModify {boneName=bicep_01_l}
    scale=<0.15 0.15 0.15>
   }
   "Transform (Modify) Bone - Bone: pectoralis_02_l"(ComponentPose=@n3, Alpha=@n2, Translation="-8.560,0.000,-0.000", Rotation="0,0,0", Scale="1,1,1") {
    rotationMode=BMM_Additive
    rotationSpace=BCS_BoneSpace
    scaleMode=BMM_Additive
    scaleSpace=BCS_BoneSpace
    translationMode=BMM_Additive
    translationSpace=BCS_BoneSpace
    alphaBoolBlend {blendOption=Linear}
    boneToModify {boneName=pectoralis_02_l}
    translation {x=-8.5597, z=0}
   }
   "Transform (Modify) Bone - Bone: bicep_02_l"(Alpha=@n2, Translation="0,0,0", Rotation="0,0,0", Scale="0.15,0.15,0.15") {
    rotationMode=BMM_Additive
    rotationSpace=BCS_BoneSpace
    scaleMode=BMM_Additive
    scaleSpace=BCS_BoneSpace
    alphaBoolBlend {blendOption=Linear}
    boneToModify {boneName=bicep_02_l}
    scale=<0.15 0.15 0.15>
   }
   n4 = "Constraint - pectoralis_02_r_aim (PPPPPP)"!(
    ComponentPose=Local To Component(LocalPose=Copy Pose From Mesh() {bUseAttachedParent=true})
    Alpha=1
    ConstraintWeights_0=0.0027
    ConstraintWeights_1:float=0.4332
    ConstraintWeights_2:float=0.0459
    ConstraintWeights_3:float=0.1137
    ConstraintWeights_4:float=0.0841
    ConstraintWeights_5:float=0.3203
   ) -> (Pose) {
    alphaBoolBlend {blendOption=Linear}
    boneToModify {boneName=pectoralis_02_r_aim}
    constraintSetup [
     {targetBone {boneName=upperarm_r}}
     {targetBone {boneName=upperarm_twist_01_r}}
     {targetBone {boneName=pectoralis_02_r}}
     {targetBone {boneName=spine_04}}
     {targetBone {boneName=spine_05}}
     {targetBone {boneName=clavicle_r}}
    ]
    constraintWeights [1,1,1,1,1,1]
   }
   n5 = "Constraint - pectoralis_02_l_aim (PPPPPP)"!(
    ComponentPose=@n4
    Alpha=1
    ConstraintWeights_0=0.3203
    ConstraintWeights_1:float=0.0459
    ConstraintWeights_2:float=0.1137
    ConstraintWeights_3:float=0.0841
    ConstraintWeights_4:float=0.4332
    ConstraintWeights_5:float=0.0027
   ) -> (Pose) {
    alphaBoolBlend {blendOption=Linear}
    boneToModify {boneName=pectoralis_02_l_aim}
    constraintSetup [
     {targetBone {boneName=clavicle_l}}
     {targetBone {boneName=pectoralis_02_l}}
     {targetBone {boneName=spine_04}}
     {targetBone {boneName=spine_05}}
     {targetBone {boneName=upperarm_twist_01_l}}
     {targetBone {boneName=upperarm_l}}
    ]
    constraintWeights [1,1,1,1,1,1]
   }
   n6 = Look At - Bone: pectoralis_01_l!(ComponentPose=@n5, Alpha=1, LookAtLocation:Vector="0,0,0") -> (Pose) {
    alphaBoolBlend {blendOption=Linear}
    boneToModify {boneName=pectoralis_01_l}
    lookAtLocation {x=0}
    lookAtTarget {boneReference {boneName=pectoralis_02_l_aim}}
    lookAt_Axis {axis=<1 0>}
    lookUp_Axis {axis {x=0, z=1}}
   }
   n7 = Look At - Bone: pectoralis_01_r!(ComponentPose=@n6, Alpha=1, LookAtLocation:Vector="0,0,0") -> (Pose) {
    alphaBoolBlend {blendOption=Linear}
    boneToModify {boneName=pectoralis_01_r}
    lookAtLocation {x=0}
    lookAtTarget {boneReference {boneName=pectoralis_02_r_aim}}
    lookAt_Axis {axis=<1 0>}
    lookUp_Axis {axis {x=0, z=1}}
   }
   n8 = Look At - Bone: scapula_00_l(ComponentPose=@n7, Alpha=1) {
    alphaBoolBlend {blendOption=Linear}
    boneToModify {boneName=scapula_00_l}
    lookAtLocation {x=0}
    lookAtTarget {boneReference {boneName=scapula_01_l_aim}}
    lookAt_Axis {axis=<1 0>}
    lookUp_Axis {axis {x=0, z=1}}
   }
   n9 = Look At - Bone: lattissimus_00_l(ComponentPose=@n8, Alpha=1) {
    alphaBoolBlend {blendOption=Linear}
    boneToModify {boneName=lattissimus_00_l}
    lookAtLocation {x=0}
    lookAtTarget {boneReference {boneName=upperarm_l}}
    lookAt_Axis {axis=<1 0>}
    lookUp_Axis {axis {x=0, z=1}}
   }
   n10 = Look At - Bone: scapula_00_r(ComponentPose=@n9, Alpha=1) {
    alphaBoolBlend {blendOption=Linear}
    boneToModify {boneName=scapula_00_r}
    lookAtLocation {x=0}
    lookAtTarget {boneReference {boneName=scapula_01_r_aim}}
    lookAt_Axis {axis=<1 0>}
    lookUp_Axis {axis {x=0, z=1}}
   }
   n11 = Look At - Bone: lattissimus_00_r(ComponentPose=@n10, Alpha=1) {
    alphaBoolBlend {blendOption=Linear}
    boneToModify {boneName=lattissimus_00_r}
    lookAtLocation {x=0}
    lookAtTarget {boneReference {boneName=upperarm_r}}
    lookAt_Axis {axis=<1 0>}
    lookUp_Axis {axis {x=0, z=1}}
   }
   n12 = Get RlatStretchAlphaRight() -> (RlatStretchAlphaRight:double)
   n13 = "Transform (Modify) Bone - Bone: tricep_01_r"(ComponentPose=@n11, Alpha=@n12, Translation="-0.000,-0.5,0.000", Rotation="0,0,0", Scale="1.000,0.5,1.000") {
    rotationMode=BMM_Additive
    rotationSpace=BCS_BoneSpace
    scaleMode=BMM_Additive
    scaleSpace=BCS_BoneSpace
    translationMode=BMM_Additive
    translationSpace=BCS_BoneSpace
    alphaBoolBlend {blendOption=Linear}
    boneToModify {boneName=tricep_01_r}
    scale {y=0.5}
    translation {y=-0.5}
   }
   n14 = "Transform (Modify) Bone - Bone: tricep_02_r"(ComponentPose=@n13, Alpha=@n12, Translation="0.000,-3,-0.000", Rotation="0,0,0", Scale="1.000,0.6,1.000") {
    rotationMode=BMM_Additive
    rotationSpace=BCS_BoneSpace
    scaleMode=BMM_Additive
    scaleSpace=BCS_BoneSpace
    translationMode=BMM_Additive
    translationSpace=BCS_BoneSpace
    alphaBoolBlend {blendOption=Linear}
    boneToModify {boneName=tricep_02_r}
    scale {y=0.6}
    translation {y=-3}
   }
   n15 = "Transform (Modify) Bone - Bone: lattissimus_01_r"(ComponentPose=@n14, Alpha=@n12, Translation="0.000,-2,0.000", Rotation="0,0,0", Scale="1,1,1") {
    rotationMode=BMM_Additive
    rotationSpace=BCS_BoneSpace
    scaleMode=BMM_Additive
    scaleSpace=BCS_BoneSpace
    translationMode=BMM_Additive
    translationSpace=BCS_BoneSpace
    alphaBoolBlend {blendOption=Linear}
    boneToModify {boneName=lattissimus_01_r}
    translation {y=-2}
   }
   n16 = Get RlatStretchAlphaLeft() -> (RlatStretchAlphaLeft:double)
   n17 = "Transform (Modify) Bone - Bone: tricep_01_r"(ComponentPose=@n15, Alpha=@n16, Translation="-0.000,0.5,0.000", Rotation="0,0,0", Scale="1.000,0.5,1.000") {
    rotationMode=BMM_Additive
    rotationSpace=BCS_BoneSpace
    scaleMode=BMM_Additive
    scaleSpace=BCS_BoneSpace
    translationMode=BMM_Additive
    translationSpace=BCS_BoneSpace
    alphaBoolBlend {blendOption=Linear}
    boneToModify {boneName=tricep_01_r}
    scale {y=0.5}
    translation {y=0.5}
   }
   n18 = "Transform (Modify) Bone - Bone: tricep_02_r"(ComponentPose=@n17, Alpha=@n16, Translation="0.000,3,-0.000", Rotation="0,0,0", Scale="1.000,0.6,1.000") {
    rotationMode=BMM_Additive
    rotationSpace=BCS_BoneSpace
    scaleMode=BMM_Additive
    scaleSpace=BCS_BoneSpace
    translationMode=BMM_Additive
    translationSpace=BCS_BoneSpace
    alphaBoolBlend {blendOption=Linear}
    boneToModify {boneName=tricep_02_r}
    scale {y=0.6}
    translation {y=3}
   }
   n19 = "Transform (Modify) Bone - Bone: lattissimus_01_r"(ComponentPose=@n18, Alpha=@n16, Translation="0.000,2,0.000", Rotation="0,0,0", Scale="1,1,1") {
    rotationMode=BMM_Additive
    rotationSpace=BCS_BoneSpace
    scaleMode=BMM_Additive
    scaleSpace=BCS_BoneSpace
    translationMode=BMM_Additive
    translationSpace=BCS_BoneSpace
    alphaBoolBlend {blendOption=Linear}
    boneToModify {boneName=lattissimus_01_r}
    translation {y=2}
   }
   n20 = Get PectStretchAlphaRight() -> (PectStretchAlphaRight:double)
   n21 = "Transform (Modify) Bone - Bone: bicep_01_r"(ComponentPose=@n19, Alpha=@n20, Translation="0.000,0.000,-3.225", Rotation="0,0,0", Scale="0.750,0.750,0.750") {
    rotationMode=BMM_Additive
    rotationSpace=BCS_BoneSpace
    scaleMode=BMM_Additive
    scaleSpace=BCS_BoneSpace
    translationMode=BMM_Additive
    translationSpace=BCS_BoneSpace
    alphaBoolBlend {blendOption=Linear}
    boneToModify {boneName=bicep_01_r}
    scale=<0.75 0.75 0.75>
    translation {z=-3.225}
   }
   n22 = "Transform (Modify) Bone - Bone: bicep_02_r"(ComponentPose=@n21, Alpha=@n20, Translation="0,0,0", Rotation="0,0,0", Scale="1,1,1") {
    rotationMode=BMM_Additive
    rotationSpace=BCS_BoneSpace
    scaleMode=BMM_Additive
    scaleSpace=BCS_BoneSpace
    translationMode=BMM_Additive
    translationSpace=BCS_BoneSpace
    alphaBoolBlend {blendOption=Linear}
    boneToModify {boneName=bicep_02_r}
   }
   n23 = "Transform (Modify) Bone - Bone: pectoralis_02_r"(ComponentPose=@n22, Alpha=@n20, Translation="-2.594,2.870,-0.000", Rotation="0,0,0", Scale="1.000,1.000,0.500") {
    rotationMode=BMM_Additive
    rotationSpace=BCS_BoneSpace
    scaleMode=BMM_Additive
    scaleSpace=BCS_BoneSpace
    translationMode=BMM_Additive
    translationSpace=BCS_BoneSpace
    alphaBoolBlend {blendOption=Linear}
    boneToModify {boneName=pectoralis_02_r}
    scale {z=0.5}
    translation=<-2.594 2.87>
   }
   Output Pose(Result=Component To Local(ComponentPose=@n23)) {layerGroup=None}
  ]}
 EventGraph {nodes [
   Event BlueprintUpdateAnimation()
   n1 = GetPartSkeletalMeshComponent()
   > n2 = TransformFromBoneSpace!(
    execute
    self:Skinned Mesh Component Object Reference=@n1
    BoneName:Name=dyn_tricep_measure_r
    InPosition:Vector="0, 0, 0"
    InRotation:Rotator="0, 0, 0"
   ) -> (then, OutPosition:Vector, OutRotation:Rotator)
   > n3 = TransformFromBoneSpace(self=@n1, BoneName=dyn_lat_measure_r, InPosition="0, 0, 0", InRotation="0, 0, 0")
   > Set RlatStretchAlphaRight(RlatStretchAlphaRight:double=MapRangeClamped(Value="Distance (Vector)"!(V1:Vector=@n2, V2:Vector=@n3) -> (ReturnValue:double), InRangeA=18, InRangeB=12)) -> (Output_Get:double)
   n4 = GetPartSkeletalMeshComponent()
   > n5 = TransformFromBoneSpace(self=@n4, BoneName=dyn_tricep_measure_r, InPosition="0, 0, 0", InRotation="0, 0, 0")
   > n6 = TransformFromBoneSpace(self=@n4, BoneName=dyn_pec_measure_r, InPosition="0, 0, 0", InRotation="0, 0, 0")
   > Set PectStretchAlphaRight(PectStretchAlphaRight:double=MapRangeClamped(Value="Distance (Vector)"(V1=@n5, V2=@n6), InRangeA=20, InRangeB=9)) -> (Output_Get:double)
   n7 = GetPartSkeletalMeshComponent()
   > n8 = TransformFromBoneSpace(self=@n7, BoneName=dyn_tricep_measure_l, InPosition="0, 0, 0", InRotation="0, 0, 0")
   > n9 = TransformFromBoneSpace(self=@n7, BoneName=dyn_lat_measure_l, InPosition="0, 0, 0", InRotation="0, 0, 0")
   > Set RlatStretchAlphaLeft(RlatStretchAlphaLeft:double=MapRangeClamped(Value="Distance (Vector)"(V1=@n8, V2=@n9), InRangeA=18, InRangeB=12)) -> (Output_Get:double)
   n10 = GetPartSkeletalMeshComponent()
   > n11 = TransformFromBoneSpace(self=@n10, BoneName=dyn_tricep_measure_l, InPosition="0, 0, 0", InRotation="0, 0, 0")
   > n12 = TransformFromBoneSpace(self=@n10, BoneName=dyn_pec_measure_l, InPosition="0, 0, 0", InRotation="0, 0, 0")
   > Set PectStretchAlphaLeft(PectStretchAlphaLeft:double=MapRangeClamped(Value="Distance (Vector)"(V1=@n11, V2=@n12), InRangeA=20, InRangeB=9)) -> (Output_Get:double)
  ]}
]
variables [
 PectStretchAlphaLeft {type=number}
 PectStretchAlphaRight {type=number}
 RlatStretchAlphaLeft {type=number}
 RlatStretchAlphaRight {type=number}
]
