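# Anim Blueprint for the M_MED_Plum_Retro body part mesh that copies its pose from the attached parent mesh straight to the output.
# The graph also holds left and right trapezius Look At nodes and translation constraints that target aim bones, with constraint alphas read from stretch values.
# These corrective nodes form separate chains that do not connect to the output pose, and the update event is empty.
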AnimBlueprint /BRCosmetics/Characters/Player/Male/Medium/Bodies/M_MED_Plum_Retro/Meshes/Parts/M_MED_Plum_Retro_AnimBP
target_skeleton=&/M_MED_Plum_Retro_Skeleton
graphs [
 AnimGraph {nodes [
   n1 = Look At - Bone: dyn_trapezius_01_r(ComponentPose=Local To Component(), Alpha=1) {
    alphaBoolBlend {blendOption=Linear}
    boneToModify {boneName=dyn_trapezius_01_r}
    lookAtLocation {x=0}
    lookAtTarget {boneReference {boneName=dyn_trapezius_02_r_aim}}
    lookAt_Axis {axis=<1 0>}
    lookUp_Axis {axis {x=0, z=1}}
   }
   n2 = Look At - Bone: dyn_trapezius_01_l(ComponentPose=@n1, Alpha=1) {
    alphaBoolBlend {blendOption=Linear}
    boneToModify {boneName=dyn_trapezius_01_l}
    lookAtLocation {x=0}
    lookAtTarget {boneReference {boneName=dyn_trapezius_02_l_aim}}
    lookAt_Axis {axis=<1 0>}
    lookUp_Axis {axis {x=0, z=1}}
   }
   Component To Local(ComponentPose=@n2)
   n3 = "Constraint - dyn_trapezius_02_l (T)"(Alpha=Get LeftTrapsStretchAlpha() -> (LeftTrapsStretchAlpha:double), ConstraintWeights_0=1) {
    alphaBoolBlend {blendOption=Linear}
    boneToModify {boneName=dyn_trapezius_02_l}
    constraintSetup [{offsetOption=None, transformType=Translation, targetBone {boneName=dyn_trapezius_02_l_aim}}]
    constraintWeights [1]
   }
   "Constraint - dyn_trapezius_02_r (T)"(ComponentPose=@n3, Alpha=Get RightTrapsStretchAlpha() -> (RightTrapsStretchAlpha:double), ConstraintWeights_0=1) {
    alphaBoolBlend {blendOption=Linear}
    boneToModify {boneName=dyn_trapezius_02_r}
    constraintSetup [{offsetOption=None, transformType=Translation, targetBone {boneName=dyn_trapezius_02_r_aim}}]
    constraintWeights [1]
   }
   Output Pose(Result=Copy Pose From Mesh() {bUseAttachedParent=true}) {layerGroup=None}
  ]}
 EventGraph {nodes [Event BlueprintUpdateAnimation()]}
]
variables []
